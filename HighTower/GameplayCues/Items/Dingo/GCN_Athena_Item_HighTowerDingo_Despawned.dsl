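; Gameplay cue burst handler that plays the despawn effects for the HighTower Dingo item.
; On burst, it casts the target to BGA_Athena_SuperDingo and triggers that actor's despawn material VFX.
; It then attaches the first spawned Cascade particle component to the actor's mesh at the FX_Root socket, provided both components are valid.
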(fn OnBurst(Cascade)
  (params (MyTarget :type Actor%20Object%20Reference) (Parameters :type Gameplay%20Cue%20Parameters%20Structure%20%28by%20ref%29) (ParticleComponents :type Array%20of%20Cascade%20Particle%20System%20Component%20Object%20References) (AudioComponents :type Array%20of%20Audio%20Component%20Object%20References) (BurstCameraShakeInstance :type Camera%20Shake%20Base%20Object%20Reference) (BurstDecalInstance :type Decal%20Actor%20Object%20Reference))
  (bind _output (Utilities|Array|Get(acopy) ParticleComponents))
  (bind _asbga_athena_super_dingo (Utilities|Casting|CastToBGA_Athena_SuperDingo MyTarget))
  (Class|BGAAthenaDingoParent|DeSpawnMaterialVFX _asbga_athena_super_dingo)
  (Utilities|IsValid _output
    (:Is%20Valid
      (Utilities|IsValid (Variables|/Game/Athena/Items/Consumables/SuperDingo/BGA_Athena_SuperDingo.BGA_Athena_SuperDingo_C|GetMesh _asbga_athena_super_dingo)
        (:Is%20Valid
          (Transformation|AttachComponentToComponent _output (Variables|/Game/Athena/Items/Consumables/SuperDingo/BGA_Athena_SuperDingo.BGA_Athena_SuperDingo_C|GetMesh _asbga_athena_super_dingo) "FX_Root" "SnapToTarget" "KeepRelative" "KeepRelative" true)
          (return))))))
